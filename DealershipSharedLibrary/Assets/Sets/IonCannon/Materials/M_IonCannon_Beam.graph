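# Translucent, unlit, two-sided beam material for the Ion Cannon set, built for Niagara sprites and mesh particles.
# Two panning billow noises distort the UVs, from which a symmetric gradient forms a soft glow plus hard core.
# Dynamic parameters drive core width, contraction and lengthwise fades.
# The mask blends outer to inner color, adds scrolling cloud noise, and feeds emissive and opacity, both tinted by particle color, with optional depth fade.
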Material /DealershipSharedLibrary/Assets/Sets/IonCannon/Materials/M_IonCannon_Beam
graph {
 properties {
  bAutomaticallySetUsageInEditor=false
  bUsedWithMeshParticles=true
  bUsedWithNiagaraMeshParticles=true
  bUsedWithNiagaraSprites=true
  blendMode=BLEND_Translucent
  shadingModel=MSM_Unlit
  twoSided=true
  usedShadingModels=Unlit
 }
 nodes [
  n1 = ParticleColor()
  n2 = ScalarParameter() {group=Noise, parameterName=Secondary Noise Speed X}
  n3 = ScalarParameter() {defaultValue=-1, group=Noise, parameterName=Secondary Noise Speed Y}
  n4 = TextureSample(UVs=Panner(Coordinate=TextureCoordinate(), Speed=AppendVector(A=@n2, B=@n3)) {bFractionalPart=true, speedY=-1}) {texture=/DealershipSharedLibrary/FX/Textures/Noise/T_VCVFX_Clouds_RGB_Soft_03_RGB}
  n5 = VertexColor()
  n6 = TextureCoordinate()
  n7 = AppendVector(A=ScalarParameter() {parameterName=UV Offset X}, B=ScalarParameter() {parameterName=UV Offset Y})
  n8 = Multiply(
   A=TextureCoordinate()
   B=AppendVector(
    A=ScalarParameter() {defaultValue=1, group=Noise, parameterName=Noise UVScale X}
    B=ScalarParameter() {defaultValue=5, group=Noise, parameterName=Noise UVScale Y}
   )
  )
  n9 = Add(A=Multiply(A=Time(), B=ScalarParameter() {defaultValue=5, group=Noise, parameterName=Noise Scroll Speed}), B=@n5.B)
  n10 = ScalarParameter() {defaultValue=0.15, group=Noise, parameterName=Noise 1 Speed X}
  n11 = ScalarParameter() {defaultValue=-0.01, group=Noise, parameterName=Noise 1 Speed Y}
  n12 = TextureSample(UVs=Panner(Coordinate=@n8, Time=@n9, Speed=AppendVector(A=@n10, B=@n11)) {bFractionalPart=true, speedX=0.15, speedY=-0.01}) {texture=/DealershipSharedLibrary/FX/Textures/Noise/T_VCVFX_BillowNoiseTiledOffset01_RGB}
  n13 = ScalarParameter() {defaultValue=0.1, group=Noise, parameterName=Noise 2 Speed X}
  n14 = ScalarParameter() {defaultValue=0.01, group=Noise, parameterName=Noise 2 Speed Y}
  n15 = TextureSample(UVs=Panner(Coordinate=@n8, Time=@n9, Speed=AppendVector(A=@n13, B=@n14)) {bFractionalPart=true, speedX=0.1, speedY=0.01}) {texture=&T_VCVFX_BillowNoiseTiledOffset01_RGB}
  n16 = Multiply(
   A=AppendVector(A=Add(A=@n12.R, B=@n15.R), B=Add(A=@n12.G, B=@n15.G))
   B=ScalarParameter() {defaultValue=0.05, group=Distortion, parameterName=Distortion Strength}
  )
  n17 = TextureCoordinate()
  n18 = Multiply(A=@n16, B=OneMinus(None=ComponentMask(None=@n17) {g=true}))
  n19 = Clamp(None=Add(A=Add(A=@n6, B=@n7), B=StaticSwitchParameter(True=@n18, False=@n16) {parameterName=Mask Vertical Distortion}))
  n20 = ComponentMask(None=@n19) {g=true}
  n21 = StaticSwitchParameter(True=ComponentMask(None=@n19) {r=true}, False=@n20) {defaultValue=true, parameterName=Flip Axis}
  n22 = OneMinus(None=@n21)
  n23 = Multiply(A=Multiply(A=@n22, B=@n21), B=Constant() {r=4})
  n24 = Multiply(
   A=ScalarParameter() {defaultValue=0.125, group=Beam, parameterName=Beam Glow Mult}
   B=Power(Base=@n23, Exp=ScalarParameter() {defaultValue=3, group=Beam, parameterName=Beam Glow Pow})
  )
  n25 = DynamicParameter() {defaultValue=<rgba 0.025 -1 1 1>, paramNames [Beam Core Width, Beam Contract, Beam Down Mask, Param4]}
  n26 = Constant() {r=1}
  n27 = If(A=@n23, B=OneMinus(None=@n25.Beam Core Width), "A > B"=@n26, "A == B"=@n26, "A < B"=Constant())
  n28 = Multiply(
   A=@n5.A
   B=SmoothStep(
    Min=ScalarParameter() {parameterName=SmoothStep Min}
    Max=ScalarParameter() {defaultValue=0.25, parameterName=SmoothStep Max}
    Value=Add(A=@n24, B=@n27)
   )
  )
  n29 = Multiply(A=@n28, B=DepthFade(Opacity=Constant() {r=1}, FadeDistance=ScalarParameter() {group=Depth Bias, parameterName=Depth Fade Distance}))
  n30 = Clamp(None=Multiply(A=Subtract(A=@n21, B=@n25.Beam Contract), B=Subtract(A=@n22, B=@n25.Beam Contract))) {maxDefault=0.2}
  n31 = Constant() {r=5}
  n32 = Clamp(None=Multiply(A=StaticSwitchParameter(True=@n29, False=@n28) {group=Depth Bias, parameterName=Use Depth Bias}, B=Multiply(A=@n30, B=@n31)))
  n33 = Multiply(
   A=Power(Base=Multiply(A=@n4.R, B=@n32), Exp=ScalarParameter() {defaultValue=2, group=Noise, parameterName=Secondary Noise Power})
   B=ScalarParameter() {defaultValue=1, group=Noise, parameterName=Secondary Noise Intensity}
  )
  n34 = VectorParameter() {parameterName=Outer Color, defaultValue=<rgba 1 0.0009 0 1>}
  n35 = VectorParameter() {parameterName=Inner Color, defaultValue=<rgba 1 0.227 0 0>}
  n36 = Clamp(None=Multiply(A=ScalarParameter() {defaultValue=5, group=Beam, parameterName=Beam Intensity}, B=LinearInterpolate(A=@n34.RGBA, B=@n35.RGBA, Alpha=@n32))) {maxDefault=1000}
  Multiply(
   A=@n1.RGB
   B=MaterialFunctionCall(Emissive=Add(A=@n33, B=@n36)) {materialFunction=&MF_FX_EmissiveAdjustments}.Result
  )
  n37 = LinearInterpolate(A=@n32, B=@n23, Alpha=ScalarParameter() {parameterName=Width Opacity Switch, sliderMax=1}) {constAlpha=0}
  n38 = Add(A=OneMinus(None=@n20), B=@n25.Beam Down Mask)
  Multiply(A=@n1.A, B=Multiply(A=Multiply(A=@n37, B=@n38), B=Add(A=@n20, B=@n25.Param4)))
 ]
}
